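# Niagara system for the PunkKoi emote that bursts a small cloud of petal sprites from a single emitter.
# Petals spawn at once in an offset, stretched hemisphere, launch outward from a point, then drift under drag, curl noise and gravity.
# A sine on sprite width fakes tumbling alongside rotation, while a curve grows and shrinks each petal.
# A thresholded factor from scalability distance fraction also scales sprite size, collapsing petals with distance.
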NiagaraSystem /BRCosmetics/Animation/Game/MainPlayer/Emotes/PunkKoi/FX/NS_PunkKoi_PetalBurst
dependencies {usedDataInterfaces [&NiagaraDataInterfaceCurve], usedDynamicInputs [], usedModules [&SystemState]}
emitters {Petals {
  emitterUpdateScript [
   &EmitterState(Life Cycle Mode=::NewEnumerator1, Loop Behavior=::NewEnumerator1, Loop Duration=1, UseLoopDelay=false)
   &SpawnBurst_Instantaneous(Spawn Count=20, Use Spawn Probability=false, UseLoopCountLimit=false)
  ]
  particleSpawnScript [
   &InitializeParticle(
    Color Mode=::NewEnumerator0
    Lifetime=&RandomRangeFloat(Minimum=2, Maximum=1)
    Sprite Rotation Angle=&RandomRangeFloat(Minimum=0, Maximum=1)
    Sprite Rotation Mode=::NewEnumerator1
    Sprite Size Mode=::NewEnumerator1
    Uniform Sprite Size Max=7
    Uniform Sprite Size Min=10
    UsePositionOffset=false
    Write Material Random=true
   )
   &SphereLocation(
    Hemisphere X=true
    MaskSpawnGroup=false
    Non Uniform Scale=<2 1 1>
    Offset=<-3 0 0>
    Override Seed=false
    Sphere Radius=1
    Surface Only=false
    UseNonUniformScale=true
    UseOffset=true
    UseSphereOrientation=false
   )
   &AddVelocityFromPoint(
    Module.Use Falloff Distance=false
    Origin Offset=<-3 0 0>
    Velocity Strength=&RandomRangeFloat(Minimum=25, Maximum=150)
   )
  ]
  particleUpdateScript [
   /SetVariables(Particles.DistanceThreshold:NiagaraFloat=&MakeCustomFloatFromBool(
     Bool=&SetBoolByFloatComparison(
      A=0.3
      B=&Power(
       a=&ClampFloat(Float=&Subtract_Float(
         A=&Multiply_Float(A=9.1, B=&OneMinusFloat(Float=$Emitter.Scalability.DistanceFraction))
         B=8
        ))
       b=4
      )
      Comparison Type=::NewEnumerator6
     )
     True Float=&Power(
      a=&ClampFloat(Float=&Subtract_Float(
        A=&Multiply_Float(A=9.1, B=&OneMinusFloat(Float=$Emitter.Scalability.DistanceFraction))
        B=8
       ))
      b=4
     )
     False Float=0
    )) {isSetParameters=true}
   &Drag(Drag=2, Use Linear Drag=true, Use Rotational Drag=false)
   /Niagara/Modules/Update/Forces/CurlNoiseForce(Noise Strength=250, Pan Noise=true, Pan Noise Field=<0 0 -50>)
   &GravityForce(Gravity=&RandomRangeVector(Minimum=<0 0 -40>, Maximum=<0 0 -80>))
   &SolveForcesAndVelocity(Clamp Velocity=false, Limit Acceleration=false)
   &ParticleState()
   &SpriteRotationRate(Rotation Rate=&RandomRangeFloat(Minimum=50, Maximum=250))
   &ScaleSpriteSize(Scale Factor=&Multiply_Vector2(
     A=&Multiply_Vector2(
      A=&MakeVector2D(
       X=&Multiply_Float(
        A=1
        B=&Sine(
         Normalized Angle=$Emitter.Age
         Period=&Multiply_Float(A=&RandomRangeFloat(Minimum=1, Maximum=2), B=&Add_Float(A=0.2, B=0.6))
         Scale=1
         Bias=0
        )
       )
       Y=1
      )
      B=&Vector2DFromFloat(Value=&FloatFromCurve(FloatCurve=DataInterface() {propertyValues {Curve {keys [
            {arriveTangent=6.8239, interpMode=RCIM_Cubic, leaveTangent=6.8239, tangentMode=RCTM_User, time=0, value=0}
            <at 0.2 1>
            {interpMode=RCIM_Cubic, tangentMode=RCTM_User, time=0.7, value=1}
            {tangentMode=RCTM_User, time=1, value=0}
           ]}}}))
     )
     B=&Vector2DFromFloat(Value=$Particles.DistanceThreshold)
    )) {module=ScaleSpriteSize001}
   &CameraOffset(Camera Offset Amount=3)
  ]
  rendererClasses [&NiagaraSpriteRendererProperties]
  renderers [{rendererClass=&NiagaraSpriteRendererProperties}]
 }}
summary {
 systemName=NS_PunkKoi_PetalBurst
 emitters [{emitterName=Petals, rendererClasses [&NiagaraSpriteRendererProperties]}]
 userVariables []
}
