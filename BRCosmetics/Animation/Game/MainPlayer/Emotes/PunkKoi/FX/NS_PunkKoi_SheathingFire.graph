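# Niagara effect for the PunkKoi emote, a brief teal-green fiery flash apparently accompanying a weapon sheathing.
# Flame-lick, spark and petal sprites fly outward with drag, the latter two swirled by curl noise, alongside color-shifting mesh energy spikes.
# A one-shot flare, pulsing lens flare and expanding ring add the flash, with light renderers on both flares.
# Petals flutter through sine-scaled width and shrink away by scalability distance.
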NiagaraSystem /BRCosmetics/Animation/Game/MainPlayer/Emotes/PunkKoi/FX/NS_PunkKoi_SheathingFire
dependencies {
 usedDataInterfaces [&NiagaraDataInterfaceVectorCurve, &NiagaraDataInterfaceColorCurve, &NiagaraDataInterfaceCurve]
 usedDynamicInputs []
 usedModules [&SystemState]
}
emitters {
 FlameLicks {
  emitterUpdateScript [
   &EmitterState()
   &SpawnRate(SpawnRate=10, Use Spawn Probability=false)
   &SpawnBurst_Instantaneous(Spawn Count=&RandomRangeInt(Minimum=3, Maximum=5), Use Spawn Probability=false)
  ]
  particleSpawnScript [
   /Niagara/Modules/Spawn/Initialization/InitializeParticle(
    Color=<rgba 5 3.0056 1.6206 0.75>
    Lifetime=&UniformRangedFloat(Minimum=0.4, Maximum=0.5)
    Module.Write Color=false
    Module.Write Lifetime=true
    Module.Write Mass=false
    Module.Write Material Random=true
    Module.Write Position=false
    Module.Write Scale=false
    Module.Write SpriteRotation=true
    Module.Write SpriteSize=true
    Position=&SimulationPosition()
    Sprite Rotation=&UniformRangedFloat(Minimum=0, Maximum=360)
    Sprite Size=&Vector2DFromFloat(Value=&RandomRangeFloat(Minimum=25, Maximum=45))
   )
   &Color()
   &SphereLocation(
    MaskSpawnGroup=false
    Override Seed=false
    Sphere Radius=1
    Surface Only=false
    UseNonUniformScale=false
    UseOffset=false
    UseSphereOrientation=false
   )
   &AddVelocityFromPoint(
    Module.Use Falloff Distance=false
    Origin Offset=<-2 0 0>
    Velocity Strength=&RandomRangeFloat(Minimum=200, Maximum=100)
   )
   &DynamicMaterialParameters(
    Index 0 Param 2=&UniformRangedFloat(Minimum=0, Maximum=1)
    Index 0 Param 3=1
    Index 0 Param 4=&UniformRangedFloat(Minimum=1, Maximum=2)
    Module.Param0WriteEnabled=false
    Module.Param1WriteEnabled=true
    Module.Param2WriteEnabled=false
    Module.Param3WriteEnabled=true
   ) {module=DynamicMaterialParameters002}
   &CameraOffset(Camera Offset Amount=&RandomRangeFloat(Minimum=5, Maximum=10))
   &ApplyInitialForces(Apply Rotational Force to Rotational Velocity=false)
  ]
  particleUpdateScript [
   &ParticleState()
   &ScaleColor(
    Scale Mode=::NewEnumerator1
    Scale RGBA=&Vector4FromCurve(Vector4Curve=DataInterface() {propertyValues {
       AlphaCurve {keys [<at 0.0024 0>,<at 0.2852 0.6>,<at 0.5798 0.3>,<at 1 0>]}
       BlueCurve {keys [<at 0 0.0244>,<at 0.1771 2.4682>,<at 0.5441 1.1028>,<at 1 0>]}
       GreenCurve {keys [<at 0 1>,<at 0.1771 4>,<at 0.5441 1.2674>,<at 1 1>]}
       RedCurve {keys [<at 0 0>,<at 0.1771 0.701>,<at 0.5441 0.2234>,<at 1 0.1399>]}
      }})
    ScaleRGBA=true
   )
   &ScaleSpriteSize(Scale Factor=&Vector2DFromFloat(Value=&FloatFromCurve(FloatCurve=DataInterface() {propertyValues {Curve {keys [
          {arriveTangent=3.3081, interpMode=RCIM_Cubic, leaveTangent=3.3081, tangentMode=RCTM_User, time=0, value=0.1}
          {arriveTangent=0.1045, interpMode=RCIM_Cubic, leaveTangent=0.1045, tangentMode=RCTM_User, time=0.4, value=1}
          {arriveTangent=-0.8944, interpMode=RCIM_Cubic, leaveTangent=-0.8944, tangentMode=RCTM_User, time=1, value=0.8}
         ]}}})))
   &DynamicMaterialParameters(
    Index 0 Param 1=&FloatFromCurve(FloatCurve=DataInterface() {propertyValues {Curve {keys [{arriveTangent=2.7551, interpMode=RCIM_Cubic, leaveTangent=2.7551, tangentMode=RCTM_User, time=0, value=0}, <at 0.7 1>]}}})
    Index 0 Param 2=1
    Index 0 Param 3=2
    Index 0 Param 4=1
    Module.Param0WriteEnabled=true
    Module.Param1WriteEnabled=true
    Module.Param2WriteEnabled=true
    Module.Param3WriteEnabled=false
   ) {module=DynamicMaterialParameters001}
   &Drag(Drag=4, Use Linear Drag=true, Use Rotational Drag=false)
   &SolveForcesAndVelocity(Clamp Velocity=false, Limit Acceleration=false)
   &SubUVAnimation() {module=SubUVAnimation001}
  ]
  rendererClasses [&NiagaraSpriteRendererProperties]
  renderers [{rendererClass=&NiagaraSpriteRendererProperties}]
 }
 Flash_Flare {
  emitterUpdateScript [
   &EmitterState()
   &SpawnBurst_Instantaneous(Spawn Count=1, Use Spawn Probability=false)
  ]
  particleSpawnScript [
   &InitializeParticle(
    Color=<rgba 0.4241 2 1.3646 1>
    Lifetime=0.1
    Module.Write Color=true
    Module.Write Lifetime=true
    Module.Write Mass=true
    Module.Write Material Random=true
    Module.Write Position=false
    Module.Write Scale=false
    Module.Write SpriteRotation=false
    Module.Write SpriteSize=true
    Position=&SimulationPosition()
    Sprite Size=&Vector2DFromFloat(Value=80)
   )
   &DynamicMaterialParameters(
    Index 0 Param 1=0.2
    Index 0 Param 2=1
    Module.Param0WriteEnabled=true
    Module.Param1WriteEnabled=true
    Module.Param2WriteEnabled=true
    Module.Param3WriteEnabled=true
   )
   &CameraOffset(Camera Offset Amount=20)
  ]
  particleUpdateScript [
   &ParticleState()
   &ScaleColor(
    Scale Alpha=&Lerp_Float(A=0.1, B=0, Alpha=$Particles.NormalizedAge)
    Scale RGB=&VectorFromFloat(Value=8)
    ScaleA=true
    ScaleRGB=true
   )
   &FN_Light_v2(Light Exponent=3, Light Radius=50, LightEnabled=true, Scale Color=1)
  ]
  rendererClasses [&NiagaraSpriteRendererProperties, &NiagaraLightRendererProperties]
  renderers [
   {rendererClass=&NiagaraSpriteRendererProperties}
   {rendererClass=&NiagaraLightRendererProperties, rendererIndex=1}
  ]
 }
 Lensflare {
  emitterUpdateScript [
   &EmitterState()
   &SpawnBurst_Instantaneous(Use Spawn Probability=false, UseLoopCountLimit=false)
  ]
  particleSpawnScript [
   /Niagara/Modules/Spawn/Initialization/V2/InitializeParticle(
    Color=<rgba 0.1914 1 0.7044 0.5>
    Lifetime=0.5
    Position=<2 0 0>
    Position Mode=::NewEnumerator1
    Sprite Size Mode=::NewEnumerator3
    Uniform Sprite Size=60
    UsePositionOffset=false
    Write Material Random=true
   )
   /SetVariables(Particles.HaloColor:LinearColor=<rgba 1 1 1 0.5>) {isSetParameters=true}
  ]
  particleUpdateScript [
   &ParticleState(Kill Particles When Lifetime Has Elapsed=false, Let Infinitely Lived Particles Die When Emitter Deactivates=true)
   &ScaleSpriteSize(
    Scale Sprite Size Mode=::NewEnumerator0
    Uniform Scale Factor=&Multiply_Float(
     A=&FloatFromCurve(FloatCurve=DataInterface() {propertyValues {Curve {keys [
          {interpMode=RCIM_Cubic, tangentMode=RCTM_User, time=0, value=0.4}
          {interpMode=RCIM_Cubic, tangentMode=RCTM_User, time=0.5, value=1}
          {tangentMode=RCTM_User, time=1, value=0}
         ]}}})
     B=&Multiply_Float(A=1, B=&Sine(Normalized Angle=$Emitter.Age, Period=0.2, Scale=0.2, Bias=1))
    )
   )
   &FN_Light_v2(
    Light Exponent=3
    Light Radius=125
    LightEnabled=true
    Scale Color:NiagaraFloat=&FloatFromCurve(FloatCurve=DataInterface() {propertyValues {Curve {keys [
         {interpMode=RCIM_Cubic, tangentMode=RCTM_User, time=0, value=0}
         {arriveTangent=1.3333, leaveTangent=1.3333, time=0.25, value=1}
         {arriveTangent=-1.3333, interpMode=RCIM_Cubic, leaveTangent=-1.3333, time=0.75, value=1}
         {tangentMode=RCTM_User, time=1, value=0}
        ]}}})
   )
   &CameraOffset(Camera Offset Amount=3)
   /SetVariables(Particles.HaloSize:Vector2f=&Multiply_Vector2D_ByFloat(
     Float=&Multiply_Float(A=1, B=&Sine(Normalized Angle=$Emitter.Age, Period=0.2, Scale=0.05, Bias=1))
     Vector2D=<70 70>
    )) {isSetParameters=true}
  ]
  rendererClasses [&NiagaraSpriteRendererProperties, &NiagaraLightRendererProperties]
  renderers [
   {rendererClass=&NiagaraSpriteRendererProperties}
   {rendererClass=&NiagaraLightRendererProperties, rendererIndex=1}
  ]
 }
 Petals {
  emitterUpdateScript [
   &EmitterState(Life Cycle Mode=::NewEnumerator1, Loop Behavior=::NewEnumerator1, Loop Duration=1, UseLoopDelay=false)
   &SpawnBurst_Instantaneous(Spawn Count=10, Use Spawn Probability=false, UseLoopCountLimit=false)
  ]
  particleSpawnScript [
   &InitializeParticle(
    Color Mode=::NewEnumerator0
    Lifetime=&RandomRangeFloat(Minimum=2, Maximum=1)
    Sprite Rotation Angle=&RandomRangeFloat(Minimum=0, Maximum=1)
    Sprite Rotation Mode=::NewEnumerator1
    Sprite Size Mode=::NewEnumerator1
    Uniform Sprite Size Max=7
    Uniform Sprite Size Min=10
    UsePositionOffset=false
    Write Material Random=true
   )
   &SphereLocation(
    MaskSpawnGroup=false
    Override Seed=false
    Sphere Radius=2
    Surface Only=false
    UseNonUniformScale=false
    UseOffset=false
    UseSphereOrientation=false
   )
   &AddVelocityFromPoint(
    Module.Use Falloff Distance=false
    Origin Offset=<-3 0 0>
    Velocity Strength=&RandomRangeFloat(Minimum=50, Maximum=200)
   )
  ]
  particleUpdateScript [
   /SetVariables(Particles.DistanceThreshold:NiagaraFloat=&MakeCustomFloatFromBool(
     Bool=&SetBoolByFloatComparison(
      A=0.3
      B=&Power(
       a=&ClampFloat(Float=&Subtract_Float(
         A=&Multiply_Float(A=9.1, B=&OneMinusFloat(Float=$Emitter.Scalability.DistanceFraction))
         B=8
        ))
       b=4
      )
      Comparison Type=::NewEnumerator6
     )
     True Float=&Power(
      a=&ClampFloat(Float=&Subtract_Float(
        A=&Multiply_Float(A=9.1, B=&OneMinusFloat(Float=$Emitter.Scalability.DistanceFraction))
        B=8
       ))
      b=4
     )
     False Float=0
    )) {isSetParameters=true}
   &Drag(Drag=2, Use Linear Drag=true, Use Rotational Drag=false)
   /Niagara/Modules/Update/Forces/CurlNoiseForce(Noise Frequency=0.1, Noise Strength=150, Pan Noise=true, Pan Noise Field=<0 0 -50>)
   &GravityForce(Gravity=&RandomRangeVector(Minimum=<0 0 -20>, Maximum=<0 0 -80>))
   &SolveForcesAndVelocity(Clamp Velocity=false, Limit Acceleration=false)
   &ParticleState()
   &SpriteRotationRate(Rotation Rate=&RandomRangeFloat(Minimum=50, Maximum=250))
   &ScaleSpriteSize(Scale Factor=&Multiply_Vector2(
     A=&Multiply_Vector2(
      A=&MakeVector2D(
       X=&Multiply_Float(
        A=1
        B=&Sine(
         Normalized Angle=$Emitter.Age
         Period=&Multiply_Float(A=&RandomRangeFloat(Minimum=1, Maximum=2), B=&Add_Float(A=0.2, B=0.6))
         Scale=1
         Bias=0
        )
       )
       Y=1
      )
      B=&Vector2DFromFloat(Value=&FloatFromCurve(FloatCurve=DataInterface() {propertyValues {Curve {keys [
            {arriveTangent=6.8239, interpMode=RCIM_Cubic, leaveTangent=6.8239, tangentMode=RCTM_User, time=0, value=0}
            <at 0.2 1>
            {interpMode=RCIM_Cubic, tangentMode=RCTM_User, time=0.7, value=1}
            {tangentMode=RCTM_User, time=1, value=0}
           ]}}}))
     )
     B=&Vector2DFromFloat(Value=$Particles.DistanceThreshold)
    )) {module=ScaleSpriteSize001}
  ]
  rendererClasses [&NiagaraSpriteRendererProperties]
  renderers [{rendererClass=&NiagaraSpriteRendererProperties}]
 }
 Ring {
  emitterUpdateScript [
   &EmitterState(
    Enable Distance Culling=true
    Life Cycle Mode=::NewEnumerator1
    Loop Behavior=::NewEnumerator1
    Loop Delay=0.0333
    Loop Duration=1
    Max Distance Response=::KillImmediately
    MaxDistance=3500
    Module.bUseMaxDistance=true
    Module.bUseMinDistance=false
    Scalability Mode=::NewEnumerator1
    UseLoopDelay=true
   )
   &SpawnBurst_Instantaneous(Spawn Count=1, Use Spawn Probability=false)
  ]
  particleSpawnScript [
   &InitializeParticle(
    Lifetime=0.15
    Sprite Rotation Mode=::NewEnumerator2
    Sprite Size Mode=::NewEnumerator3
    Uniform Sprite Size=100
    UsePositionOffset=false
    Write Material Random=true
   )
   &CameraOffset(Camera Offset Amount=20)
  ]
  particleUpdateScript [
   &ParticleState()
   &ScaleColor(
    Scale Alpha=&FloatFromCurve(FloatCurve=DataInterface() {propertyValues {Curve {keys [<at 0 0.5>,<at 1 0>]}}})
    Scale RGB=&MakeVectorFromLinearColorRGB(Color=&ColorFromCurve(ColorCurve=DataInterface() {propertyValues {
        AlphaCurve {keys [<at 0 1 RCIM_Cubic>,<at 1 0 RCIM_Cubic>]}
        BlueCurve {keys [<at 0 0.3907 RCIM_Cubic>,<at 1 1 RCIM_Cubic>]}
        GreenCurve {keys [<at 0 1 RCIM_Cubic>,<at 1 0.4331 RCIM_Cubic>]}
        RedCurve {keys [<at 0 0.122 RCIM_Cubic>,<at 1 0.0937 RCIM_Cubic>]}
       }}))
    ScaleA=true
    ScaleRGB=true
   )
   &ScaleSpriteSize(Scale Factor=&Vector2DFromFloat(Value=&FloatFromCurve(FloatCurve=DataInterface() {propertyValues {Curve {keys [{arriveTangent=2.7513, interpMode=RCIM_Cubic, leaveTangent=2.7513, tangentMode=RCTM_User, time=0, value=0}, <at 1 1>]}}})))
  ]
  rendererClasses [&NiagaraSpriteRendererProperties]
  renderers [{rendererClass=&NiagaraSpriteRendererProperties}]
 }
 Sparks {
  emitterUpdateScript [
   &EmitterState()
   &SpawnRate(SpawnRate=20, Use Spawn Probability=false)
   &SpawnBurst_Instantaneous(Spawn Count=15, Use Spawn Probability=false, UseLoopCountLimit=false)
  ]
  particleSpawnScript [
   &InitializeParticle(
    Color Mode=::NewEnumerator0
    Lifetime Max=0.5
    Lifetime Min=0.2
    Lifetime Mode=::NewEnumerator1
    Sprite Size Mode=::NewEnumerator1
    Uniform Sprite Size Max=1.5
    Uniform Sprite Size Min=1
    UsePositionOffset=false
    Write Material Random=true
   )
   &AddVelocity(Velocity=&RandomRangeVector(Minimum=<50 0 0>, Maximum=<150 1 1>))
   &SphereLocation(
    MaskSpawnGroup=false
    Override Seed=false
    Sphere Radius=5
    Surface Only=false
    UseNonUniformScale=false
    UseOffset=false
    UseSphereOrientation=false
   )
   &AddVelocityFromPoint(Module.Use Falloff Distance=false, Origin Offset=<-5 0 0>, Velocity Strength=90)
  ]
  particleUpdateScript [
   /Niagara/Modules/Update/Forces/V2/CurlNoiseForce(
    Noise Frequency=90
    Noise Strength=&FloatFromCurve(
     FloatCurve=DataInterface() {propertyValues {Curve {keys [
         {arriveTangent=4.1785, interpMode=RCIM_Cubic, leaveTangent=4.1785, tangentMode=RCTM_User, time=0, value=0}
         {arriveTangent=0.9995, interpMode=RCIM_Cubic, leaveTangent=0.9995, tangentMode=RCTM_User, time=0.2223, value=0.6825}
         {arriveTangent=-0.0405, interpMode=RCIM_Cubic, leaveTangent=-0.0405, tangentMode=RCTM_User, time=1, value=1}
        ]}}}
     Scale Curve=1000
    )
    Pan Noise=true
    Pan Noise Field=<0 0 1>
   )
   &Drag(Drag=4, Use Linear Drag=true, Use Rotational Drag=false)
   &SolveForcesAndVelocity(Clamp Velocity=false, Limit Acceleration=false)
   &ParticleState()
   &ScaleSpriteSizeBySpeed(
    Max Scale Factor=<1 5>
    Min Scale Factor=<0.5 4>
    Sample Scale Factor By Curve=true
    Scale Factor Curve=DataInterface() {propertyValues {Curve {keys [{tangentMode=RCTM_User, time=0, value=0}, {tangentMode=RCTM_User, time=1, value=1}]}}}
    Velocity Threshold=200
   )
   &ScaleSpriteSize(Uniform Curve Sprite Scale=DataInterface() {propertyValues {Curve {keys [
        {arriveTangent=6.1033, interpMode=RCIM_Cubic, leaveTangent=6.1033, tangentMode=RCTM_User, time=0, value=0}
        {arriveTangent=0.3392, interpMode=RCIM_Cubic, leaveTangent=0.3392, tangentMode=RCTM_User, time=0.3, value=0.95}
        {arriveTangent=-0.0405, interpMode=RCIM_Cubic, leaveTangent=-0.0405, tangentMode=RCTM_User, time=1, value=0.5}
       ]}}})
   &ScaleColor(
    Scale Alpha=&FloatFromCurve(FloatCurve=DataInterface() {propertyValues {Curve {keys [{interpMode=RCIM_Cubic, tangentMode=RCTM_User, time=0, value=1}, {interpMode=RCIM_Cubic, tangentMode=RCTM_User, time=1, value=0}]}}})
    Scale RGB=&MakeVectorFromLinearColorRGB(Color=&ColorFromCurve(ColorCurve=DataInterface() {propertyValues {
        AlphaCurve {keys [<at 0 1 RCIM_Cubic>,<at 1 1 RCIM_Cubic>]}
        BlueCurve {keys [<at 0.3037 0.8498 RCIM_Cubic>,<at 0.831 0.1842>,<at 1 0>]}
        GreenCurve {keys [<at 0.3037 1 RCIM_Cubic>,<at 0.831 1>,<at 1 0>]}
        RedCurve {keys [<at 0.3037 0.2135 RCIM_Cubic>,<at 0.831 0.5224>,<at 1 0>]}
       }}))
    ScaleA=true
    ScaleRGB=true
   )
   &CameraOffset(Camera Offset Amount=2)
  ]
  rendererClasses [&NiagaraSpriteRendererProperties]
  renderers [{rendererClass=&NiagaraSpriteRendererProperties}]
 }
 SpikeEnergy {
  emitterUpdateScript [
   &EmitterState(Life Cycle Mode=::NewEnumerator1, Loop Behavior=::NewEnumerator1, Loop Duration=0.2, UseLoopDelay=false)
   &SpawnRate(SpawnRate=20, Use Spawn Probability=false)
   &SpawnBurst_Instantaneous(Spawn Count=5, Use Spawn Probability=false, UseLoopCountLimit=false)
  ]
  particleSpawnScript [
   &InitializeParticle(
    Color=<rgba 1 1 1 0.4>
    Lifetime=&RandomRangeFloat(Minimum=0.25, Maximum=0.35)
    Mesh Scale=&RandomRangeVector(Minimum=<0.6 0.3 0.3>, Maximum=<0.4 0.3 0.3>)
    Mesh Scale Mode=::NewEnumerator4
    UsePositionOffset=false
    Write Material Random=true
   )
   &InitialMeshOrientation(
    Mesh Orientation Mode=::NewEnumerator2
    Rotation=&RandomRangeVector(Minimum=<0 0.15 0.15>, Maximum=<1 -0.15 -0.15>)
    Use Rotation Vector=true
   )
  ]
  particleUpdateScript [
   &ParticleState()
   &ScaleMeshSize(Scale Factor=&VectorFromCurve(VectorCurve=DataInterface() {propertyValues {
       XCurve {keys [{interpMode=RCIM_Cubic, tangentMode=RCTM_User, time=0, value=0}, {interpMode=RCIM_Cubic, tangentMode=RCTM_User, time=0.5, value=0.8}]}
       YCurve {keys [{arriveTangent=6.5531, interpMode=RCIM_Cubic, leaveTangent=6.5531, tangentMode=RCTM_User, time=0, value=0}, <at 0.5 1>, <at 1 0.2>]}
       ZCurve {keys [{arriveTangent=7.1759, interpMode=RCIM_Cubic, leaveTangent=7.1759, tangentMode=RCTM_User, time=0, value=0}, <at 0.2 1>, <at 1 0.2>]}
      }}))
   &ScaleColor(
    Scale Alpha=&FloatFromCurve(FloatCurve=DataInterface() {propertyValues {Curve {keys [
         {arriveTangent=-2.6258, interpMode=RCIM_Cubic, leaveTangent=-2.6258, tangentMode=RCTM_User, time=0, value=1}
         {interpMode=RCIM_Cubic, tangentMode=RCTM_User, time=1, value=0}
        ]}}})
    Scale RGB=&MakeVectorFromLinearColorRGB(Color=&ColorFromCurve(ColorCurve=DataInterface() {propertyValues {
        AlphaCurve {keys [<at 0 1 RCIM_Cubic>,<at 1 1 RCIM_Cubic>]}
        BlueCurve {keys [
          <at 0.1792 1.5016>
          {arriveTangent=-0.7623, interpMode=RCIM_Cubic, leaveTangent=-0.7623, time=0.2961, value=7.2573}
          <at 0.8372 1 RCIM_Cubic>
         ]}
        GreenCurve {keys [<at 0.1792 10>, {arriveTangent=-13.9828, interpMode=RCIM_Cubic, leaveTangent=-13.9828, time=0.2961, value=10}, <at 0.8372 0.799 RCIM_Cubic>]}
        RedCurve {keys [
          <at 0.1792 7.3528>
          {arriveTangent=-10.929, interpMode=RCIM_Cubic, leaveTangent=-10.929, time=0.2961, value=1.8924}
          <at 0.8372 0.1612 RCIM_Cubic>
         ]}
       }}))
    ScaleA=true
    ScaleRGB=true
   )
   &CameraOffset(Camera Offset Amount=1)
  ]
  rendererClasses [&NiagaraMeshRendererProperties]
  renderers [{rendererClass=&NiagaraMeshRendererProperties, rendererIndex=0}]
 }
}
summary {
 systemName=NS_PunkKoi_SheathingFire
 emitters [
  {emitterName=SpikeEnergy, rendererClasses [&NiagaraMeshRendererProperties]}
  {emitterName=Lensflare, rendererClasses [&NiagaraSpriteRendererProperties, &NiagaraLightRendererProperties]}
  {emitterName=Sparks, rendererClasses [&NiagaraSpriteRendererProperties]}
  {emitterName=FlameLicks, rendererClasses [&NiagaraSpriteRendererProperties]}
  {emitterName=Flash_Flare, rendererClasses [&NiagaraSpriteRendererProperties, &NiagaraLightRendererProperties]}
  {emitterName=Petals, rendererClasses [&NiagaraSpriteRendererProperties]}
  {emitterName=Ring, rendererClasses [&NiagaraSpriteRendererProperties]}
 ]
 userVariables [User.SkeletalMesh {
   type=NiagaraDataInterfaceSkeletalMesh
   defaultValue {
    struct=NiagaraExt_VariableValue_DataInterface
    value {dataInterface=&self:NiagaraDataInterfaceSkeletalMesh_0}
   }
  }]
}
